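# Sound cue for the YorkMan emitter in the Battle Pass S18 York quest, layering a continuous ambient loop with intermittent laughs.
# A Mixer combines the looping wave, faded in by an Enveloper over half a second, with an endlessly looped Delay into Random branch that picks among five laughs every several seconds.
# The laughs use their own attenuation node, separate from the cue-level attenuation, and the cue uses the quest sound class.
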SoundCue /BattlePassS18/QuestActors/Assets/Sound/York/Quest_York_02_YorkMan_Emit_Cue
root=Mixer_0
nodes [
 {
  id=Attenuation_0
  children [Looping_0]
  properties {attenuationSettings=&/Quest_York_02_Laugh_Atten}
 }
 {id=Delay_0, children [Random_0], properties {delayMax=12, delayMin=8}}
 {
  id=Enveloper_0
  children [WavePlayer_0]
  properties {pitchCurve {editorCurveData {keys [{time=0}]}}, volumeCurve {editorCurveData {keys [<at 0 0>,<at 0.5 1>]}}}
 }
 {id=Looping_0, children [Delay_0]}
 {id=Mixer_0, children [Enveloper_0, Attenuation_0], properties {inputVolume [1,1]}}
 {id=Random_0, children [WavePlayer_1,WavePlayer_2,WavePlayer_3,WavePlayer_4,WavePlayer_5], properties {weights [1,1,1,1,1]}}
 {id=WavePlayer_0, properties {bLooping=true, soundWaveAssetPtr=&/Quest_York_02_YorkMan_Emit_LP}}
 {id=WavePlayer_1, properties {soundWaveAssetPtr=&/Quest_York_02_YorkMan_Laugh_01}}
 {id=WavePlayer_2, properties {soundWaveAssetPtr=&/Quest_York_02_YorkMan_Laugh_02}}
 {id=WavePlayer_3, properties {soundWaveAssetPtr=&/Quest_York_02_YorkMan_Laugh_03}}
 {id=WavePlayer_4, properties {soundWaveAssetPtr=&/Quest_York_02_YorkMan_Laugh_04}}
 {id=WavePlayer_5, properties {soundWaveAssetPtr=&/Quest_York_02_YorkMan_Laugh_05}}
]
properties {
 attenuationSettings=&/Quest_York_02_YorkMan_Atten
 maxDistance=4050
 soundClassObject=&Fort_Quests
}
