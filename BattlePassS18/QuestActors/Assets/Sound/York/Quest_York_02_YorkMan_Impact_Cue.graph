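# Sound cue for the York quest's YorkMan character being hit, layering a pain vocal over an impact sound.
# A Mixer blends two equally weighted Random branches, one picking from eight "Ow" vocal waves and the other from five impact waves, with the vocal branch mixed slightly louder.
# It plays through the YorkMan attenuation asset and the Fort_Quests sound class.
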SoundCue /BattlePassS18/QuestActors/Assets/Sound/York/Quest_York_02_YorkMan_Impact_Cue
root=Mixer_1
nodes [
 {id=Mixer_1, children [Random_0,Random_1], properties {inputVolume [1.2,1]}}
 {
  id=Random_0
  children [WavePlayer_0,WavePlayer_1,WavePlayer_2,WavePlayer_3,WavePlayer_4,WavePlayer_5,WavePlayer_6,WavePlayer_7]
  properties {weights [1,1,1,1,1,1,1,1]}
 }
 {id=Random_1, children [WavePlayer_8, WavePlayer_9, WavePlayer_10, WavePlayer_11, WavePlayer_12], properties {weights [1,1,1,1,1]}}
 {id=WavePlayer_0, properties {soundWaveAssetPtr=&/Quest_York_02_YorkMan_Ow_01}}
 {id=WavePlayer_1, properties {soundWaveAssetPtr=&/Quest_York_02_YorkMan_Ow_02}}
 {id=WavePlayer_10, properties {soundWaveAssetPtr=&/Quest_York_02_YorkMan_Impact_03}}
 {id=WavePlayer_11, properties {soundWaveAssetPtr=&/Quest_York_02_YorkMan_Impact_04}}
 {id=WavePlayer_12, properties {soundWaveAssetPtr=&/Quest_York_02_YorkMan_Impact_05}}
 {id=WavePlayer_2, properties {soundWaveAssetPtr=&/Quest_York_02_YorkMan_Ow_03}}
 {id=WavePlayer_3, properties {soundWaveAssetPtr=&/Quest_York_02_YorkMan_Ow_04}}
 {id=WavePlayer_4, properties {soundWaveAssetPtr=&/Quest_York_02_YorkMan_Ow_05}}
 {id=WavePlayer_5, properties {soundWaveAssetPtr=&/Quest_York_02_YorkMan_Ow_06}}
 {id=WavePlayer_6, properties {soundWaveAssetPtr=&/Quest_York_02_YorkMan_Ow_07}}
 {id=WavePlayer_7, properties {soundWaveAssetPtr=&/Quest_York_02_YorkMan_Ow_08}}
 {id=WavePlayer_8, properties {soundWaveAssetPtr=&/Quest_York_02_YorkMan_Impact_01}}
 {id=WavePlayer_9, properties {soundWaveAssetPtr=&/Quest_York_02_YorkMan_Impact_02}}
]
properties {
 attenuationSettings=&/Quest_York_02_YorkMan_Atten
 maxDistance=2050
 soundClassObject=&Fort_Quests
 volumeMultiplier=0.9
}
