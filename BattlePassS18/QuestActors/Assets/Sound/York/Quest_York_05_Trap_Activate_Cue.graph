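# Sound cue for the activation of a trap in the York quest 05 content of Battle Pass Season 18.
# A single wave player feeds the trap activate sound wave straight to the output.
# It uses the York trap attenuation asset and the Fort_Quests sound class, with a concurrency override that caps it at two simultaneous instances.
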SoundCue /BattlePassS18/QuestActors/Assets/Sound/York/Quest_York_05_Trap_Activate_Cue
root=WavePlayer_0
nodes [{id=WavePlayer_0, properties {soundWaveAssetPtr=&/Quest_York_05_Trap_Activate}}]
properties {
 attenuationSettings=&/Quest_York_Trap_Atten
 bOverrideConcurrency=true
 maxDistance=4050
 soundClassObject=&Fort_Quests
 volumeMultiplier=1
 concurrencyOverrides {maxCount=2}
}
