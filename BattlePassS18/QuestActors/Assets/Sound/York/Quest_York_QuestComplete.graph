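# Sound cue for the quest-complete sting of the York quest in Battle Pass Season 18, routed to the quests sound class.
# A mixer layers a dedicated quest-complete wave, slightly delayed, with a quieter reused reward-screen chest loot upgrade sound.
# The cue overrides concurrency to cap simultaneous instances and sets a very large max distance.
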SoundCue /BattlePassS18/QuestActors/Assets/Sound/York/Quest_York_QuestComplete
root=Mixer_0
nodes [
 {id=Delay_0, children [WavePlayer_0], properties {delayMin=0.05}}
 {id=Mixer_0, children [Delay_0,WavePlayer_1], properties {inputVolume [1,0.3]}}
 {id=WavePlayer_0, properties {soundWaveAssetPtr=&/Quest_York_QuestComplete_01}}
 {id=WavePlayer_1, properties {soundWaveAssetPtr=/Game/Sounds/Fort_UI_Sounds/Rewards/Fort_RewardScreen_TeamScore_ChestLootUpgrade}}
]
properties {
 bOverrideConcurrency=true
 maxDistance=2097152
 soundClassObject=&Fort_Quests
 concurrencyOverrides {maxCount=2}
}
